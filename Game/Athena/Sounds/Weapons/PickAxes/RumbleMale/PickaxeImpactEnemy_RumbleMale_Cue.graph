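# Sound cue for RumbleMale pickaxe impacts on enemies, played through the player-impact attenuation and Fort_Impacts sound class.
# After a fixed short delay, a Branch on the FruitSelection bool parameter picks one of two equally weighted random pools of six waves.
# True or unset plays the Impact_B variations; false plays the Impact_D variations.
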SoundCue /Game/Athena/Sounds/Weapons/PickAxes/RumbleMale/PickaxeImpactEnemy_RumbleMale_Cue
root=Delay_0
nodes [
 {id=Branch_1, children [Random_0,Random_1,Random_0], properties {boolParameterName=FruitSelection}}
 {id=Delay_0, children [Branch_1], properties {delayMax=0.1, delayMin=0.1}}
 {
  id=Random_0
  children [WavePlayer_6, WavePlayer_13, WavePlayer_14, WavePlayer_15, WavePlayer_16, WavePlayer_17]
  properties {weights [1,1,1,1,1,1]}
 }
 {
  id=Random_1
  children [WavePlayer_7, WavePlayer_8, WavePlayer_9, WavePlayer_10, WavePlayer_11, WavePlayer_12]
  properties {weights [1,1,1,1,1,1]}
 }
 {id=WavePlayer_10, properties {soundWaveAssetPtr=&/Pickaxe_RumbleMale_Impact_D_04}}
 {id=WavePlayer_11, properties {soundWaveAssetPtr=&/Pickaxe_RumbleMale_Impact_D_05}}
 {id=WavePlayer_12, properties {soundWaveAssetPtr=&/Pickaxe_RumbleMale_Impact_D_06}}
 {id=WavePlayer_13, properties {soundWaveAssetPtr=&/Pickaxe_RumbleMale_Impact_B_02}}
 {id=WavePlayer_14, properties {soundWaveAssetPtr=&/Pickaxe_RumbleMale_Impact_B_03}}
 {id=WavePlayer_15, properties {soundWaveAssetPtr=&/Pickaxe_RumbleMale_Impact_B_04}}
 {id=WavePlayer_16, properties {soundWaveAssetPtr=&/Pickaxe_RumbleMale_Impact_B_05}}
 {id=WavePlayer_17, properties {soundWaveAssetPtr=&/Pickaxe_RumbleMale_Impact_B_06}}
 {id=WavePlayer_6, properties {soundWaveAssetPtr=&/Pickaxe_RumbleMale_Impact_B_01}}
 {id=WavePlayer_7, properties {soundWaveAssetPtr=&/Pickaxe_RumbleMale_Impact_D_01}}
 {id=WavePlayer_8, properties {soundWaveAssetPtr=&/Pickaxe_RumbleMale_Impact_D_02}}
 {id=WavePlayer_9, properties {soundWaveAssetPtr=&/Pickaxe_RumbleMale_Impact_D_03}}
]
properties {
 attenuationSettings=/Game/Sounds/Attenuation/Fort_Weapon_Pickaxe_Player_Impact_Att
 maxDistance=3512
 soundClassObject=&Fort_Impacts
 volumeMultiplier=0.6225
 concurrencySet [null]
}
